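; Test trigger actor that runs a console command, likely a cheat, when a player pawn overlaps its Cylinder component.
; The overlap casts to FortPlayerPawnAthena, falling back to FortPlayerPawn, and passes the pawn with the ConsoleCommand variable to ExecuteCommand.
; That function casts the pawn's controller to PlayerController and executes the command through it.
; BeginPlay, Tick, actor overlap and the construction script are empty.
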(event EventBeginPlay)

(event Collision|EventActorBeginOverlap
  (params (OtherActor :type Actor%20Object%20Reference)))

(event EventTick
  (params (DeltaSeconds :type Float%20%28single-precision%29)))

(event OnComponentBeginOverlap(Cylinder)
  (params (OverlappedComponent :type Primitive%20Component%20Object%20Reference) (OtherActor :type Actor%20Object%20Reference) (OtherComp :type Primitive%20Component%20Object%20Reference) (OtherBodyIndex :type Integer) (bFromSweep :type Boolean) (SweepResult :type Hit%20Result%20Structure%20%28by%20ref%29))
  (bind _consolecommand (Variables|Default|GetConsoleCommand))
  (bind _asfort_player_pawn_athena (Utilities|Casting|CastToFortPlayerPawnAthena OtherActor)
    (:then
      (CallFunction|ExecuteCommand _consolecommand _asfort_player_pawn_athena))
    (:CastFailed
      (bind _asfort_player_pawn (Utilities|Casting|CastToFortPlayerPawn OtherActor))
      (CallFunction|ExecuteCommand _consolecommand _asfort_player_pawn))))

(fn executeCommand
  (params (Command :type String) (self2 :type Pawn%20Object%20Reference))
  (bind _asplayer_controller (Utilities|Casting|CastToPlayerController (Pawn|GetController self2)))
  (Development|ExecuteConsoleCommand Command _asplayer_controller))

(fn ConstructionScript)
